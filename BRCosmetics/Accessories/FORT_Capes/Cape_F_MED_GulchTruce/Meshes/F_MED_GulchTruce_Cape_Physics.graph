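# Physics asset for the GulchTruce cape cosmetic, defining the capsule bodies and joints for the cape's dynamic bones.
# Center, left and right chains each run from a small MSTR root through seven segments with limited swing and tight twist.
# The lower four side segments tie to the center chain with free angular motion.
# Capsules on spine, pelvis, shoulders, upper arms, legs and backpack root form the body, joined by angularly free constraints.
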PhysicsAsset /BRCosmetics/Accessories/FORT_Capes/Cape_F_MED_GulchTruce/Meshes/F_MED_GulchTruce_Cape_Physics
bodies {
 backpack_root [{shapeType=Capsule, length=1, radius=1}]
 calf_l [{shapeType=Capsule, length=11.7963, radius=39.0787, center=<-26.1258 24.118 0>, rotation=<rot 90 179.8881 -178.6232>}]
 calf_r [{shapeType=Capsule, length=11.7963, radius=39.0787, center=<26.1258 -24.118 0>, rotation=<rot 90 180 1.4887>}]
 clavicle_l [{shapeType=Capsule, length=12.9688, radius=9.6562, center=<12.3638 -1.6165 0>, rotation=<rot 90 0 -6.5611>}]
 clavicle_r [{shapeType=Capsule, length=12.9688, radius=9.6562, center=<-12.3638 1.6165 0>, rotation=<rot 90 0 173.4389>}]
 dyn_cape_01 [{shapeType=Capsule, length=11.7556, radius=5.9773, shapeName=dyn_cape_01_capsule, center=<7.15 0 8>, rotation=<rot 90 0 0>}]
 dyn_cape_02 [{shapeType=Capsule, length=10, radius=10, shapeName=dyn_cape_01_capsule, center=<7.15 0 5>, rotation=<rot 90 0 0>}]
 dyn_cape_03 [{shapeType=Capsule, length=10, radius=12, shapeName=dyn_cape_01_capsule, center=<7.15 0 5>, rotation=<rot 90 0 0>}]
 dyn_cape_04 [{shapeType=Capsule, length=10, radius=12, shapeName=dyn_cape_01_capsule, center=<7.15 0 5>, rotation=<rot 90 0 0>}]
 dyn_cape_05 [{shapeType=Capsule, length=10, radius=12, shapeName=dyn_cape_01_capsule, center=<7.15 0 5>, rotation=<rot 90 0 0>}]
 dyn_cape_06 [{shapeType=Capsule, length=10, radius=12, shapeName=dyn_cape_01_capsule, center=<7.15 0 5>, rotation=<rot 90 0 0>}]
 dyn_cape_07 [{shapeType=Capsule, length=10, radius=14, shapeName=dyn_cape_01_capsule, center=<5.8445 0 5>, rotation=<rot 90 0 0>}]
 dyn_cape_MSTR [{shapeType=Capsule, length=0.505, radius=0.505, shapeName=dyn_cape_r_MSTR_capsule, rotation=<rot 0 -90 0>}]
 dyn_cape_l_01 [{shapeType=Capsule, length=11.7556, radius=5.9773, shapeName=dyn_cape_01_capsule, center=<7.15 0 8>, rotation=<rot 90 0 0>}]
 dyn_cape_l_02 [{shapeType=Capsule, length=10, radius=10, shapeName=dyn_cape_01_capsule, center=<7.15 4 5>, rotation=<rot 90 0 0>}]
 dyn_cape_l_03 [{shapeType=Capsule, length=10, radius=12, shapeName=dyn_cape_01_capsule, center=<7.15 7 5>, rotation=<rot 90 0 0>}]
 dyn_cape_l_04 [{shapeType=Capsule, length=10, radius=12, shapeName=dyn_cape_01_capsule, center=<7.15 7 5>, rotation=<rot 90 0 0>}]
 dyn_cape_l_05 [{shapeType=Capsule, length=10, radius=12, shapeName=dyn_cape_01_capsule, center=<7.15 7 5>, rotation=<rot 90 0 0>}]
 dyn_cape_l_06 [{shapeType=Capsule, length=10, radius=12, shapeName=dyn_cape_01_capsule, center=<7.15 7 5>, rotation=<rot 90 0 0>}]
 dyn_cape_l_07 [{shapeType=Capsule, length=10, radius=14, shapeName=dyn_cape_01_capsule, center=<5.8445 10 5>, rotation=<rot 90 0 0>}]
 dyn_cape_l_MSTR [{shapeType=Capsule, length=0.505, radius=0.505, shapeName=dyn_cape_r_MSTR_capsule, rotation=<rot 0 -90 0>}]
 dyn_cape_r_01 [{shapeType=Capsule, length=11.7556, radius=5.9773, shapeName=dyn_cape_01_capsule, center=<7.15 0 8>, rotation=<rot 90 0 0>}]
 dyn_cape_r_02 [{shapeType=Capsule, length=10, radius=10, shapeName=dyn_cape_01_capsule, center=<7.15 -4 5>, rotation=<rot 90 0 0>}]
 dyn_cape_r_03 [{shapeType=Capsule, length=10, radius=12, shapeName=dyn_cape_01_capsule, center=<7.15 -7 5>, rotation=<rot 90 0 0>}]
 dyn_cape_r_04 [{shapeType=Capsule, length=10, radius=12, shapeName=dyn_cape_01_capsule, center=<7.15 -7 5>, rotation=<rot 90 0 0>}]
 dyn_cape_r_05 [{shapeType=Capsule, length=10, radius=12, shapeName=dyn_cape_01_capsule, center=<7.15 -7 5>, rotation=<rot 90 0 0>}]
 dyn_cape_r_06 [{shapeType=Capsule, length=10, radius=12, shapeName=dyn_cape_01_capsule, center=<7.15 -7 5>, rotation=<rot 90 0 0>}]
 dyn_cape_r_07 [{shapeType=Capsule, length=10, radius=14, shapeName=dyn_cape_01_capsule, center=<5.8445 -10 5>, rotation=<rot 90 0 0>}]
 dyn_cape_r_MSTR [{shapeType=Capsule, length=0.505, radius=0.505, shapeName=dyn_cape_r_MSTR_capsule, rotation=<rot 0 -90 0>}]
 pelvis [{shapeType=Capsule, length=3.8685, radius=14.2252, center=<-1.9314 -1.5382 0>}]
 spine_01 [{shapeType=Capsule, length=1.371, radius=14.6602, center=<0.5427 -0.4734 0>}]
 spine_02 [{shapeType=Capsule, length=2.6493, radius=13.1632, center=<3.5949 0.2148 0>}]
 spine_03 [{shapeType=Capsule, length=39.6562, radius=10, center=<2.5875 -1 0>}]
 spine_04 [{shapeType=Capsule, length=31.4062, radius=12, center=<2.8832 0.25 0>}]
 spine_05 [{shapeType=Capsule, length=32.875, radius=12, center=<-0.0029 0.1172 0>}]
 thigh_l [{shapeType=Capsule, length=23.2484, radius=14.0641, center=<-18.6538 -0.1273 0>, rotation=<rot 90 2.6271 0>}]
 thigh_r [{shapeType=Capsule, length=23.2484, radius=14.0641, center=<18.6538 0.1273 0>, rotation=<rot 90 -177.3729 0>}]
 upperarm_l [{shapeType=Capsule, length=20.2791, radius=7.2834, center=<11.9223 -1.2712 0>, rotation=<rot 90 6.5611 0>}]
 upperarm_r [{shapeType=Capsule, length=20.2791, radius=7.2834, center=<-11.9223 1.2712 0>, rotation=<rot 90 -173.4389 0>}]
}
constraints [
 {
  bone1Name=spine_01
  bone2Name=pelvis
  swing1LimitDegrees=45
  swing1Motion=Free
  swing2LimitDegrees=45
  swing2Motion=Free
  twistLimitDegrees=45
  twistMotion=Free
 }
 {
  bone1Name=spine_02
  bone2Name=spine_01
  swing1LimitDegrees=45
  swing1Motion=Free
  swing2LimitDegrees=45
  swing2Motion=Free
  twistLimitDegrees=45
  twistMotion=Free
 }
 {
  bone1Name=spine_03
  bone2Name=spine_02
  swing1LimitDegrees=45
  swing1Motion=Free
  swing2LimitDegrees=45
  swing2Motion=Free
  twistLimitDegrees=45
  twistMotion=Free
 }
 {
  bone1Name=spine_04
  bone2Name=spine_03
  swing1LimitDegrees=45
  swing1Motion=Free
  swing2LimitDegrees=45
  swing2Motion=Free
  twistLimitDegrees=45
  twistMotion=Free
 }
 {
  bone1Name=spine_05
  bone2Name=spine_04
  swing1LimitDegrees=45
  swing1Motion=Free
  swing2LimitDegrees=45
  swing2Motion=Free
  twistLimitDegrees=45
  twistMotion=Free
 }
 {
  bone1Name=backpack_root
  bone2Name=spine_05
  swing1LimitDegrees=45
  swing1Motion=Free
  swing2LimitDegrees=45
  swing2Motion=Free
  twistLimitDegrees=45
  twistMotion=Free
 }
 {bone1Name=dyn_cape_01, bone2Name=dyn_cape_MSTR, swing1LimitDegrees=10, swing2LimitDegrees=45, twistLimitDegrees=5}
 {bone1Name=dyn_cape_02, bone2Name=dyn_cape_01, swing1LimitDegrees=20, swing2LimitDegrees=45, twistLimitDegrees=5}
 {bone1Name=dyn_cape_03, bone2Name=dyn_cape_02, swing1LimitDegrees=25, swing2LimitDegrees=45, twistLimitDegrees=5}
 {bone1Name=dyn_cape_04, bone2Name=dyn_cape_03, swing1LimitDegrees=25, swing2LimitDegrees=45, twistLimitDegrees=5}
 {bone1Name=dyn_cape_05, bone2Name=dyn_cape_04, swing1LimitDegrees=25, swing2LimitDegrees=45, twistLimitDegrees=5}
 {bone1Name=dyn_cape_06, bone2Name=dyn_cape_05, swing1LimitDegrees=25, swing2LimitDegrees=45, twistLimitDegrees=5}
 {bone1Name=dyn_cape_07, bone2Name=dyn_cape_06, swing1LimitDegrees=25, swing2LimitDegrees=45, twistLimitDegrees=5}
 {bone1Name=dyn_cape_l_01, bone2Name=dyn_cape_l_MSTR, swing1LimitDegrees=10, swing2LimitDegrees=45, twistLimitDegrees=5}
 {bone1Name=dyn_cape_l_02, bone2Name=dyn_cape_l_01, swing1LimitDegrees=20, swing2LimitDegrees=45, twistLimitDegrees=5}
 {bone1Name=dyn_cape_l_03, bone2Name=dyn_cape_l_02, swing1LimitDegrees=25, swing2LimitDegrees=45, twistLimitDegrees=5}
 {bone1Name=dyn_cape_l_04, bone2Name=dyn_cape_l_03, swing1LimitDegrees=25, swing2LimitDegrees=45, twistLimitDegrees=5}
 {bone1Name=dyn_cape_l_05, bone2Name=dyn_cape_l_04, swing1LimitDegrees=25, swing2LimitDegrees=45, twistLimitDegrees=5}
 {bone1Name=dyn_cape_l_06, bone2Name=dyn_cape_l_05, swing1LimitDegrees=25, swing2LimitDegrees=45, twistLimitDegrees=5}
 {bone1Name=dyn_cape_l_07, bone2Name=dyn_cape_l_06, swing1LimitDegrees=25, swing2LimitDegrees=45, twistLimitDegrees=5}
 {bone1Name=dyn_cape_r_01, bone2Name=dyn_cape_r_MSTR, swing1LimitDegrees=10, swing2LimitDegrees=45, twistLimitDegrees=5}
 {bone1Name=dyn_cape_r_02, bone2Name=dyn_cape_r_01, swing1LimitDegrees=20, swing2LimitDegrees=45, twistLimitDegrees=5}
 {bone1Name=dyn_cape_r_03, bone2Name=dyn_cape_r_02, swing1LimitDegrees=25, swing2LimitDegrees=45, twistLimitDegrees=5}
 {bone1Name=dyn_cape_r_04, bone2Name=dyn_cape_r_03, swing1LimitDegrees=25, swing2LimitDegrees=45, twistLimitDegrees=5}
 {bone1Name=dyn_cape_r_05, bone2Name=dyn_cape_r_04, swing1LimitDegrees=25, swing2LimitDegrees=45, twistLimitDegrees=5}
 {bone1Name=dyn_cape_r_06, bone2Name=dyn_cape_r_05, swing1LimitDegrees=25, swing2LimitDegrees=45, twistLimitDegrees=5}
 {bone1Name=dyn_cape_r_07, bone2Name=dyn_cape_r_06, swing1LimitDegrees=25, swing2LimitDegrees=45, twistLimitDegrees=5}
 {
  bone1Name=clavicle_l
  bone2Name=spine_05
  swing1LimitDegrees=45
  swing1Motion=Free
  swing2LimitDegrees=45
  swing2Motion=Free
  twistLimitDegrees=45
  twistMotion=Free
 }
 {
  bone1Name=clavicle_r
  bone2Name=spine_05
  swing1LimitDegrees=45
  swing1Motion=Free
  swing2LimitDegrees=45
  swing2Motion=Free
  twistLimitDegrees=45
  twistMotion=Free
 }
 {
  bone1Name=upperarm_l
  bone2Name=clavicle_l
  swing1LimitDegrees=45
  swing1Motion=Free
  swing2LimitDegrees=45
  swing2Motion=Free
  twistLimitDegrees=45
  twistMotion=Free
 }
 {
  bone1Name=upperarm_r
  bone2Name=clavicle_r
  swing1LimitDegrees=45
  swing1Motion=Free
  swing2LimitDegrees=45
  swing2Motion=Free
  twistLimitDegrees=45
  twistMotion=Free
 }
 {
  bone1Name=thigh_l
  bone2Name=pelvis
  swing1LimitDegrees=45
  swing1Motion=Free
  swing2LimitDegrees=45
  swing2Motion=Free
  twistLimitDegrees=45
  twistMotion=Free
 }
 {
  bone1Name=thigh_r
  bone2Name=pelvis
  swing1LimitDegrees=45
  swing1Motion=Free
  swing2LimitDegrees=45
  swing2Motion=Free
  twistLimitDegrees=45
  twistMotion=Free
 }
 {
  bone1Name=calf_l
  bone2Name=thigh_l
  swing1LimitDegrees=45
  swing1Motion=Free
  swing2LimitDegrees=45
  swing2Motion=Free
  twistLimitDegrees=45
  twistMotion=Free
 }
 {
  bone1Name=calf_r
  bone2Name=thigh_r
  swing1LimitDegrees=45
  swing1Motion=Free
  swing2LimitDegrees=45
  swing2Motion=Free
  twistLimitDegrees=45
  twistMotion=Free
 }
 {
  bone1Name=dyn_cape_l_04
  bone2Name=dyn_cape_04
  swing1LimitDegrees=45
  swing1Motion=Free
  swing2LimitDegrees=45
  swing2Motion=Free
  twistLimitDegrees=45
  twistMotion=Free
 }
 {
  bone1Name=dyn_cape_r_04
  bone2Name=dyn_cape_04
  swing1LimitDegrees=45
  swing1Motion=Free
  swing2LimitDegrees=45
  swing2Motion=Free
  twistLimitDegrees=45
  twistMotion=Free
 }
 {
  bone1Name=dyn_cape_l_05
  bone2Name=dyn_cape_05
  swing1LimitDegrees=45
  swing1Motion=Free
  swing2LimitDegrees=45
  swing2Motion=Free
  twistLimitDegrees=45
  twistMotion=Free
 }
 {
  bone1Name=dyn_cape_r_05
  bone2Name=dyn_cape_05
  swing1LimitDegrees=45
  swing1Motion=Free
  swing2LimitDegrees=45
  swing2Motion=Free
  twistLimitDegrees=45
  twistMotion=Free
 }
 {
  bone1Name=dyn_cape_l_06
  bone2Name=dyn_cape_06
  swing1LimitDegrees=45
  swing1Motion=Free
  swing2LimitDegrees=45
  swing2Motion=Free
  twistLimitDegrees=45
  twistMotion=Free
 }
 {
  bone1Name=dyn_cape_r_06
  bone2Name=dyn_cape_06
  swing1LimitDegrees=45
  swing1Motion=Free
  swing2LimitDegrees=45
  swing2Motion=Free
  twistLimitDegrees=45
  twistMotion=Free
 }
 {
  bone1Name=dyn_cape_l_07
  bone2Name=dyn_cape_07
  swing1LimitDegrees=45
  swing1Motion=Free
  swing2LimitDegrees=45
  swing2Motion=Free
  twistLimitDegrees=45
  twistMotion=Free
 }
 {
  bone1Name=dyn_cape_r_07
  bone2Name=dyn_cape_07
  swing1LimitDegrees=45
  swing1Motion=Free
  swing2LimitDegrees=45
  swing2Motion=Free
  twistLimitDegrees=45
  twistMotion=Free
 }
]
